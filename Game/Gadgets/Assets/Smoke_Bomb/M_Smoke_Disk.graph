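# Unlit, translucent, two-sided particle material for the smoke bomb's disk effect, usable on mesh and Niagara mesh particles.
# A smoke disk texture is multiplied by particle color, with its alpha times particle alpha passing through a depth fade to soften intersections.
# The tinted color runs through a day/night auto bloom control function, then an FX emissive adjustment function to produce the glow.
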Material /Game/Gadgets/Assets/Smoke_Bomb/M_Smoke_Disk
graph {
 properties {
  bUsedWithMeshParticles=true
  bUsedWithNanite=true
  bUsedWithNiagaraMeshParticles=true
  blendMode=BLEND_Translucent
  pixelDepthOffsetMode=PDOM_Legacy
  shadingModel=MSM_Unlit
  twoSided=true
  usedShadingModels=Unlit
 }
 nodes [
  n1 = ParticleColor()
  n2 = TextureSample() {texture=/Game/Effects/Fort_Effects/Textures/Smoke/T_Smoke_Bomb_Disk}
  DepthFade(Opacity=Multiply(A=@n1.A, B=@n2.A)) {fadeDistanceDefault=5}
  MaterialFunctionCall(Emissive=MaterialFunctionCall!(DayBloom_Mult=Constant() {r=1}, In=Multiply(A=@n1.RGB, B=@n2.RGB)) {materialFunction=/Game/Packages/Fortnite_UniqueMaterials_1/MH_Materials/DayNight_Auto_Bloom_Control}.Result) {materialFunction=&MF_FX_EmissiveAdjustments}
 ]
}
